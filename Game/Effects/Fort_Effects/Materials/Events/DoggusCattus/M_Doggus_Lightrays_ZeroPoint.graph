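# Additive unlit mesh-particle material for animated Zero Point light rays in the DoggusCattus event effects.
# Gradients and noise-distorted, time-panned beam and lightning samples form offset color channels, desaturated and combined with a center glow.
# A lerp toward a Ceil-stepped version stylizes the rays, then noisy V falloff, particle color and optional hue cycling shape emissive.
# Static switches add coin-bob offset and near-camera fade, while depth fade drops out on low-end platforms.
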Material /Game/Effects/Fort_Effects/Materials/Events/DoggusCattus/M_Doggus_Lightrays_ZeroPoint
graph {
 properties {
  bUsedWithMeshParticles=true
  bUsedWithNiagaraMeshParticles=true
  blendMode=BLEND_Additive
  pixelDepthOffsetMode=PDOM_Legacy
  shadingModel=MSM_Unlit
  usedShadingModels=Unlit
 }
 nodes [
  n1 = LinearInterpolate(
   A=VectorParameter() {parameterName=Color_01, defaultValue=<rgba 1 0 0.1327 1>}.RGB
   B=VectorParameter() {parameterName=Color_02, defaultValue=<rgba 0 0 1 1>}.RGB
  )
  n2 = ParticleColor()
  n3 = Constant()
  n4 = TextureCoordinate()
  n5 = Add(A=AppendVector(A=Constant() {r=-0.15}, B=@n3), B=@n4)
  n6 = Constant() {r=-0.26}
  n7 = Add(A=AppendVector(A=Constant(), B=@n3), B=@n4)
  n8 = AppendVector(
   A=TextureSample(UVs=MaterialFunctionCall(UVs=@n5, "Rotation Angle (0-1)"=@n6) {materialFunction=&CustomRotator}.Rotated Values) {texture=/Game/Effects/Fort_Effects/BluePrints/WeakPointEffect/LightGradient}.R
   B=TextureSample(UVs=MaterialFunctionCall(XY=@n7) {materialFunction=&Swizzle}.YX) {texture=&LightGradient}.R
  )
  n9 = AppendVector(A=Constant() {r=0.15}, B=@n3)
  n10 = TextureSample(UVs=MaterialFunctionCall(UVs=Add(A=@n9, B=@n4), "Rotation Angle (0-1)"=Constant() {r=-0.24}) {materialFunction=&CustomRotator}.Rotated Values) {texture=&LightGradient}
  n11 = DynamicParameter() {paramNames [Speed,Wipe,Param3,SplineWiidth]}
  n12 = Add(A=Time(), B=@n11.Speed)
  n13 = MaterialFunctionCall(Float2=TextureCoordinate() {vTiling=0.1}) {materialFunction=&BreakOutFloat2Components}
  n14 = ScalarParameter() {defaultValue=1, parameterName=U_Tile}
  n15 = TextureSample(UVs=Panner(Coordinate=TextureCoordinate() {uTiling=0.7, vTiling=0.7}) {speedY=0.15}) {samplerType=SAMPLERTYPE_LinearColor, texture=&HighContrastNoise}
  n16 = Add(A=AppendVector(A=Multiply(A=@n13.R, B=@n14), B=@n13.G), B=Multiply(A=AppendVector(A=@n15.R, B=@n15.R)) {constB=0.05})
  n17 = Add(A=AppendVector(A=Constant() {r=-0.02}, B=Multiply(A=@n12) {constB=0.2}), B=@n16)
  n18 = TextureSample(UVs=@n17) {texture=/Game/Global/Textures/Default/Masks_and_Gradients/T_BlasterBeamDistort}
  n19 = Constant() {r=1.5}
  n20 = Add(A=Time(), B=@n11.Speed)
  n21 = Add(A=AppendVector(A=Constant(), B=Multiply(A=@n20) {constB=0.2}), B=@n16)
  n22 = TextureSample(UVs=@n21) {texture=&T_BlasterBeamDistort}
  n23 = AppendVector(
   A=Multiply(
    A=Power(Base=@n18.R, Exp=@n19)
    B=TextureSample(UVs=@n17) {samplerType=SAMPLERTYPE_LinearColor, texture=/Game/Effects/Fort_Effects/World/Enviroment/Lightning/lightningPainting}.G
   )
   B=Multiply(
    A=Power(Base=@n22.R, Exp=@n19)
    B=TextureSample(UVs=@n21) {samplerType=SAMPLERTYPE_LinearColor, texture=&lightningPainting}.R
   )
  )
  n24 = Add(A=Time(), B=@n11.Speed)
  n25 = Add(A=AppendVector(A=Constant() {r=0.02}, B=Multiply(A=@n24) {constB=0.2}), B=@n16)
  n26 = Multiply(
   A=Power(Base=TextureSample(UVs=@n25) {texture=&T_BlasterBeamDistort}.R, Exp=@n19)
   B=TextureSample(UVs=@n25) {samplerType=SAMPLERTYPE_LinearColor, texture=&lightningPainting}.R
  )
  n27 = Add(A=Multiply(A=AppendVector(A=@n8, B=@n10.R)) {constB=0.5}, B=Multiply(A=AppendVector(A=@n23, B=@n26)) {constB=10})
  n28 = Constant() {r=0.26}
  n29 = TextureCoordinate()
  n30 = ScalarParameter() {defaultValue=-0.03, parameterName=Slide}
  n31 = OneMinus(None=ComponentMask(None=Add(A=@n29, B=@n30)) {g=true})
  n32 = ScalarParameter() {defaultValue=12, parameterName=Center Exponent}
  n33 = Add(
   A=Multiply(A=Desaturation(None=@n27, Fraction=@n28), B=VectorParameter() {defaultValue=<rgba 1.7205 2 1.8762 1>}.RGB)
   B=Multiply(A=Power(Base=@n31, Exp=@n32), B=ScalarParameter() {defaultValue=2, parameterName=Center Brightness})
  )
  n34 = ScalarParameter() {defaultValue=-0.7, parameterName=Ceil Bias}
  n35 = Multiply(A=Ceil(None=Add(A=@n33, B=@n34)), B=ScalarParameter() {defaultValue=3, parameterName=Ceil Brightness})
  n36 = ScalarParameter() {defaultValue=0.5, parameterName=Style Lerp}
  n37 = OneMinus(None=ComponentMask(None=TextureCoordinate()) {g=true})
  n38 = TextureCoordinate() {uTiling=3, vTiling=0}
  n39 = LinearInterpolate(Alpha=TextureSample(UVs=Panner(Coordinate=@n38) {speedX=0.2, speedY=0.1}) {texture=&T_NoisePack_01}.R) {constA=2, constB=6}
  n40 = Multiply(
   A=StaticSwitchParameter(True=Multiply(A=@n1, B=@n2.RGB), False=@n2.RGB) {parameterName=Use_ColorRange_Instead_of_RandomColor?}
   B=Multiply(A=LinearInterpolate(A=@n33, B=@n35, Alpha=@n36), B=Power(Base=@n37, Exp=@n39))
  )
  n41 = Multiply(A=Time(), B=ScalarParameter() {defaultValue=1.2, parameterName=HueShiftSpeed}) {constB=1.2}
  n42 = StaticSwitchParameter(
   True=ScalarParameter() {defaultValue=1.2, parameterName=HueShift_Fixed}
   False=Multiply(A=Frac(None=@n41), B=ScalarParameter() {defaultValue=1, parameterName=HueShift_Amount})
  ) {parameterName=Use_Fixed_HueShift_Value?}
  MaterialFunctionCall(Emissive=StaticSwitchParameter(
    True=@n40
    False=MaterialFunctionCall(Hue Shift Percentage=@n42, Texture=@n40) {materialFunction=&HueShift}.Result
   ) {parameterName=NotDiscoLights}) {materialFunction=&MF_FX_EmissiveAdjustments}
  StaticSwitchParameter(
   True=MaterialFunctionCall!(TimeScale, MaxHeight, MinHeight) {materialFunction=/Game/Playgrounds/Items/Props/Materials/MF_Creative_CoinBobble}.WorldPositionOffset
   False=Constant()
  ) {parameterName=CreativeMode_CoinBob}
  n43 = ScalarParameter() {defaultValue=250, parameterName=NearCamFade_Distance}
  n44 = DepthFade(FadeDistance=ScalarParameter() {defaultValue=50, parameterName=Depth Fade Distance})
  n45 = Constant() {r=1}
  n46 = Constant() {r=1}
  n47 = Multiply(
   A=@n2.A
   B=StaticSwitchParameter(
    True=FeatureLevelSwitch(Default=@n44, ES2=@n45, ES3_1=@n45, SM4_REMOVED=@n44)
    False=QualitySwitch(Default=@n44, Low=@n46, High=@n44, Medium=@n44)
   ) {parameterName=Use_MobilePerf_Settings}
  )
  StaticSwitchParameter(
   True=Multiply(A=MaterialFunctionCall(Distance=@n43) {materialFunction=&Near_Camera_Fade}.Result, B=@n47)
   False=@n47
  ) {parameterName=Use_NearCamFade?}
 ]
}
